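# Material Gym example material that demonstrates the MF_SphereUV material function for spherical UV manipulation.
# The function's SphereUv output samples a texture parameter, which defaults to the UV editor color grid and uses the shared world-group wrap sampler.
# A DebugMask_SphereUV static switch chooses between the raw texture color and that color blended from black by the function's solid mask.
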Material /Game/Global/Materials/MaterialGym/Materials/MaterialElements/UVs/UVMainpulation/SphereUV/M_SphereUV
graph {
 properties {bAutomaticallySetUsageInEditor=false, usedShadingModels=Default Lit}
 nodes [
  n1 = MaterialFunctionCall!(UV) {materialFunction=&/MF_SphereUV}
  n2 = TextureSampleParameter2D(UVs=@n1.SphereUv) {
   group=Sphere_UV
   parameterName=SphereUVTexture
   samplerSource=SSM_Wrap_WorldGroupSettings
   texture=/UVEditor/Textures/UVEditorColorGrid_Color
  }
  StaticSwitchParameter(True=LinearInterpolate(A=Constant3Vector() {constant=<rgba 0 0 0 1>}, B=@n2.RGB, Alpha=@n1.outputSolidMask), False=@n2.RGB) {group=Sphere_UV, parameterName=DebugMask_SphereUV}
 ]
}
